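# Animation Blueprint for the LMG power-up minigun mesh, animating firing, heat transformation and barrel spin.
# Each update reads heat and overheat state from the owning weapon, comparing heat to a curve-table threshold cached at BeginPlay.
# These drive a state machine through windup, low/high-heat firing, transformations and winddown, with overheat cutting high heat to dormant.
# An additive, accumulating Barrel roll scales with square-root heat within clamped steps, before DefaultSlot; charge-time logic is unconnected.
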AnimBlueprint /CorruptionItems/Gameplay/Items/PowerUp/LMG/Animations/BPAnim_Minigun_Powerup
target_skeleton=/CorruptionItems/Gameplay/Items/PowerUp/LMG/Meshes/Power_Up_Minigun_Gun_Skeleton
graphs [
 AnimGraph {nodes [
   AnimGraphNode_StateMachine_0 = MinigunState() {is=title AnimGraphNode_StateMachine}
   n1 = "Transform (Modify) Bone - Bone: Barrel"(
    ComponentPose=Local To Component(LocalPose=@AnimGraphNode_StateMachine_0)
    Alpha=1
    Translation="0,0,0"
    Rotation=MakeRotator(Roll=Get MinigunRotation() -> (MinigunRotation:double))
    Scale="1,1,1"
   ) {rotationMode=BMM_Additive, rotationSpace=BCS_BoneSpace, alphaBoolBlend {blendOption=Linear}, boneToModify {boneName=Barrel}}
   Output Pose(Result=Slot 'DefaultSlot'(Source=Component To Local(ComponentPose=@n1))) {layerGroup=None}
  ]}
 MinigunState {
  entry=Dormant
  path=AnimGraph.AnimGraphNode_StateMachine_0.MinigunState
  ids=id
  nodes [
   AnimStateNode_0 = Dormant() {is=title AnimStateNode}
   AnimStateNode_1 = WindDown() {is=title AnimStateNode}
   AnimStateNode_2 = Firing_LowHeat() {is=title AnimStateNode}
   AnimStateNode_3 = HeatEvent_Transformation_Reverse() {is=title AnimStateNode}
   AnimStateNode_4 = Firing_HighHeat() {is=title AnimStateNode}
   AnimStateNode_6 = Windup() {is=title AnimStateNode}
   AnimStateNode_7 = HeatEvent_Transformation() {is=title AnimStateNode}
   AnimStateTransitionNode_0 = AnimStateTransitionNode()
   AnimStateTransitionNode_1 = AnimStateTransitionNode() {bAutomaticRuleBasedOnSequencePlayerInState=true}
   AnimStateTransitionNode_10 = AnimStateTransitionNode()
   AnimStateTransitionNode_2 = AnimStateTransitionNode()
   AnimStateTransitionNode_3 = AnimStateTransitionNode() {bAutomaticRuleBasedOnSequencePlayerInState=true}
   AnimStateTransitionNode_5 = AnimStateTransitionNode() {bAutomaticRuleBasedOnSequencePlayerInState=true}
   AnimStateTransitionNode_6 = AnimStateTransitionNode()
   AnimStateTransitionNode_7 = AnimStateTransitionNode()
   AnimStateTransitionNode_8 = AnimStateTransitionNode() {bAutomaticRuleBasedOnSequencePlayerInState=true}
  ]
  links [
   {
    dst=Windup
    rule=AnimGraph.AnimGraphNode_StateMachine_0.MinigunState.AnimStateTransitionNode_0.Transition
    src=Dormant
    via=AnimStateTransitionNode_0
   }
   {
    dst=Dormant
    rule=AnimGraph.AnimGraphNode_StateMachine_0.MinigunState.AnimStateTransitionNode_10.Transition
    src=Firing_HighHeat
    via=AnimStateTransitionNode_10
   }
   {
    dst=HeatEvent_Transformation_Reverse
    rule=AnimGraph.AnimGraphNode_StateMachine_0.MinigunState.AnimStateTransitionNode_7.Transition
    src=Firing_HighHeat
    via=AnimStateTransitionNode_7
   }
   {
    dst=HeatEvent_Transformation
    rule=AnimGraph.AnimGraphNode_StateMachine_0.MinigunState.AnimStateTransitionNode_2.Transition
    src=Firing_LowHeat
    via=AnimStateTransitionNode_2
   }
   {
    dst=WindDown
    rule=AnimGraph.AnimGraphNode_StateMachine_0.MinigunState.AnimStateTransitionNode_6.Transition
    src=Firing_LowHeat
    via=AnimStateTransitionNode_6
   }
   {
    dst=Firing_HighHeat
    rule=AnimGraph.AnimGraphNode_StateMachine_0.MinigunState.AnimStateTransitionNode_3.Transition
    src=HeatEvent_Transformation
    via=AnimStateTransitionNode_3
   }
   {
    dst=Firing_LowHeat
    rule=AnimGraph.AnimGraphNode_StateMachine_0.MinigunState.AnimStateTransitionNode_8.Transition
    src=HeatEvent_Transformation_Reverse
    via=AnimStateTransitionNode_8
   }
   {
    dst=Dormant
    rule=AnimGraph.AnimGraphNode_StateMachine_0.MinigunState.AnimStateTransitionNode_5.Transition
    src=WindDown
    via=AnimStateTransitionNode_5
   }
   {
    dst=Firing_LowHeat
    rule=AnimGraph.AnimGraphNode_StateMachine_0.MinigunState.AnimStateTransitionNode_1.Transition
    src=Windup
    via=AnimStateTransitionNode_1
   }
  ]
 }
 Dormant {
  path=AnimGraph.AnimGraphNode_StateMachine_0.MinigunState.AnimStateNode_0.Dormant
  nodes [Output Animation Pose(Result=Sequence Player 'PUPMiniGun_Fire_Start_Gun'() {playRate=0, sequence=&/PUPMiniGun_Fire_Start_Gun})]
 }
 WindDown {
  path=AnimGraph.AnimGraphNode_StateMachine_0.MinigunState.AnimStateNode_1.WindDown
  nodes [Output Animation Pose(Result=Sequence Player 'PUPMiniGun_Fire_WindDown_Gun'() {bLoopAnimation=false, sequence=&/PUPMiniGun_Fire_WindDown_Gun})]
 }
 Firing_LowHeat {
  path=AnimGraph.AnimGraphNode_StateMachine_0.MinigunState.AnimStateNode_2.Firing_LowHeat
  nodes [Output Animation Pose(Result=Sequence Player 'PUPMiniGun_Fire_Gun'() {sequence=&/PUPMiniGun_Fire_Gun})]
 }
 HeatEvent_Transformation_Reverse {
  path=AnimGraph.AnimGraphNode_StateMachine_0.MinigunState.AnimStateNode_3.HeatEvent_Transformation_Reverse
  nodes [Output Animation Pose(Result=Sequence Player 'PUPMiniGun_Transform_End_Gun'() {bLoopAnimation=false, sequence=&/PUPMiniGun_Transform_End_Gun})]
 }
 Firing_HighHeat {
  path=AnimGraph.AnimGraphNode_StateMachine_0.MinigunState.AnimStateNode_4.Firing_HighHeat
  nodes [Output Animation Pose(Result=Sequence Player 'PUPMiniGun_Redline_Gun'() {sequence=&/PUPMiniGun_Redline_Gun})]
 }
 Windup {
  path=AnimGraph.AnimGraphNode_StateMachine_0.MinigunState.AnimStateNode_6.Windup
  nodes [Output Animation Pose(Result=Sequence Player 'PUPMiniGun_Fire_Start_Gun'() {bLoopAnimation=false, sequence=&PUPMiniGun_Fire_Start_Gun})]
 }
 HeatEvent_Transformation {
  path=AnimGraph.AnimGraphNode_StateMachine_0.MinigunState.AnimStateNode_7.HeatEvent_Transformation
  nodes [Output Animation Pose(Result=Sequence Player 'PUPMiniGun_Redline_Start_Gun'() {bLoopAnimation=false, sequence=&/PUPMiniGun_Redline_Start_Gun})]
 }
 Transition {
  path=AnimGraph.AnimGraphNode_StateMachine_0.MinigunState.AnimStateTransitionNode_0.Transition
  nodes [Result(bCanEnterTransition="float > float"(A=Get CurHeatPercent() -> (CurHeatPercent:double), B=0) {is=title K2Node_CallFunction})]
 }
 Transition {path=AnimGraph.AnimGraphNode_StateMachine_0.MinigunState.AnimStateTransitionNode_1.Transition, nodes [Result()]}
 Transition {
  path=AnimGraph.AnimGraphNode_StateMachine_0.MinigunState.AnimStateTransitionNode_10.Transition
  nodes [Result(bCanEnterTransition=Get IsOverheat() -> (IsOverheat:bool))]
 }
 Transition {
  path=AnimGraph.AnimGraphNode_StateMachine_0.MinigunState.AnimStateTransitionNode_2.Transition
  nodes [Result(bCanEnterTransition=Get IsSuperHeat() -> (IsSuperHeat:bool))]
 }
 Transition {path=AnimGraph.AnimGraphNode_StateMachine_0.MinigunState.AnimStateTransitionNode_3.Transition, nodes [Result()]}
 Transition {path=AnimGraph.AnimGraphNode_StateMachine_0.MinigunState.AnimStateTransitionNode_5.Transition, nodes [Result()]}
 Transition {
  path=AnimGraph.AnimGraphNode_StateMachine_0.MinigunState.AnimStateTransitionNode_6.Transition
  nodes [Result(bCanEnterTransition="float <= float"!(A:double=Get CurHeatPercent(), B:double=0) -> (ReturnValue:bool) {is=title K2Node_CallFunction})]
 }
 Transition {
  path=AnimGraph.AnimGraphNode_StateMachine_0.MinigunState.AnimStateTransitionNode_7.Transition
  nodes [Result(bCanEnterTransition=NOT Boolean(A=Get IsSuperHeat()))]
 }
 Transition {path=AnimGraph.AnimGraphNode_StateMachine_0.MinigunState.AnimStateTransitionNode_8.Transition, nodes [Result()]}
 EventGraph {nodes [
   n1 = Event BlueprintUpdateAnimation()
   > Set CurDeltaTime(CurDeltaTime:double=@n1.DeltaTimeX) -> (Output_Get:double)
   > n2 = Cast To B_Assault_LMG_Powerup_Athena!(execute, Object=GetOwningActor()) -> (then, CastFailed, AsB Assault LMG Powerup Athena:B Assault LMG Powerup Athena Object Reference) {is=title K2Node_DynamicCast}
   > Set CurHeatPercent(CurHeatPercent:double=SafeDivide(
     A=Get CurrentHeatValueRep!(self:B Assault LMG Powerup Athena Object Reference=@n2) -> (CurrentHeatValueRep:double)
     B=GetOverheatingMaxValue!(self:Fort Weapon Ranged Object Reference=@n2) -> (ReturnValue:float)
    )) -> (Output_Get:double)
   > Set IsSuperHeat(IsSuperHeat:bool="float >= float"(A=Get CurHeatPercent(), B=Get HeatEventTriggerPercentRef() -> (HeatEventTriggerPercentRef:double)) {is=title K2Node_CallFunction}) -> (Output_Get:bool)
   > Set IsOverheat(IsOverheat:bool="Equal (Enum)"(
     A:EFortWeaponOverheatState Enum=GetOverheatState!(self:Fort Weapon Ranged Object Reference=@n2) -> (ReturnValue:EFortWeaponOverheatState Enum=None)
     B:EFortWeaponOverheatState Enum=Overheated
    )) -> (Output_Get:bool)
   n3 = float * float!(
    A=Sqrt!(A:double=Get CurHeatPercent()) -> (ReturnValue:double)
    B=Get MaxMinigunRotSpeed() -> (MaxMinigunRotSpeed:double)
    C:double=Get CurDeltaTime() -> (CurDeltaTime:double)
   ) -> (ReturnValue)
   > Set MinigunRotation(MinigunRotation:double=float + float(A="Clamp (Float)"(Value=@n3, Min=0.5, Max=18), B=Get MinigunRotation()) {is=title K2Node_CommutativeAssociativeBinaryOperator}) -> (Output_Get:double)
   Event BlueprintBeginPlay() {is=title K2Node_Event}
   > Set HeatEventTriggerPercentRef(HeatEventTriggerPercentRef:double=GetValueAtLevel!(Input:"Scalable Float Structure (by ref)"=Get HeatEventTriggerPercent() -> (HeatEventTriggerPercent:Scalable Float Structure), Level:float=0) -> (ReturnValue:float)) -> (Output_Get:double)
   Set CurChargeTime(CurChargeTime:double=GetChargeTime!(self:Fort Weapon Object Reference=@n2) -> (ReturnValue:float)) -> (Output_Get:double)
   > n4 = Branch(Condition=GetIsChargingWeapon!(self:Fort Weapon Object Reference=@n2) -> (ReturnValue:bool))
   > n5 = Set LastChargingTime(LastChargingTime:double=GetTimeSeconds!() -> (ReturnValue:double)) -> (Output_Get:double)
   n6 = Set ChargeTimeOverride(execute=[@n9.then, @n5.then], ChargeTimeOverride:double=Get CurChargeTime() -> (CurChargeTime:double)) -> (Output_Get:double)
   Reroute Node(InputPin=[@n6.then, @n10.then])
   n7 = GetTimeSeconds()
   n8 = Get LastChargingTime() -> (LastChargingTime:double)
   n9 = Branch(execute=@n4.else, Condition="float <= float"(A=float - float(A=@n7, B=@n8) {is=title K2Node_CallFunction}, B=1))
   n10 = Set ChargeTimeOverride(
    execute=@n9.else
    ChargeTimeOverride=FInterpTo(Current=Get ChargeTimeOverride() -> (ChargeTimeOverride:double), Target=0, DeltaTime=Get CurDeltaTime(), InterpSpeed=100000)
   )
   GetCurrentHeatPercent!(self:B Assault LMG Powerup Athena Object Reference=@n2) -> (ReturnValue:double)
   Get ChargeTimeOverride()
  ]}
]
variables [
 ChargeTimeOverride {type=number}
 CurChargeTime {type=number}
 CurDeltaTime {type=number}
 CurHeatPercent {type=number}
 HeatEventTriggerPercent {
  type=ScalableFloat
  default {
   registryType=None
   value=1
   curve {curveTable=/CorruptionItems/DataTables/CorruptionGameplayGameData, rowName=Default.LMG.PowerUp.HeatEvent.TriggerPercent}
  }
 }
 HeatEventTriggerPercentRef {type=number}
 IsOverheat {type=boolean}
 IsSuperHeat {type=boolean}
 LastChargingTime {type=number}
 MaxMinigunRotSpeed {type=number, default=1000}
 MinigunRotation {type=number}
]
