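; Power-up LMG minigun weapon Blueprint that handles overheat, heat-event and spin-up/down feedback.
; The server replicates heat, and its rep notify drives barrel glow, loop audio and Niagara parameters.
; A replicated heat-event flag toggles a fire-mode override and heat materials, while overheating ends the event and triggers cooldown effects.
; Charge events crossfade spin audio through a reversible volume timeline, and firing toggles muzzle FX and stamps fire time on dynamic materials.
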(event EventBeginPlay
  (|CallParentFunction)
  (Networking|SwitchHasAuthority
    (:Authority
      ;[1 1 1] Replicates the heat value and calls FX scripting on rep notify.
      (Overheating|AssignOnOverheatValueChanged (delegate Custom|UpdateCurrentHeatValue)))
    (:Remote
      (CallFunction|SetAudioPerspective))))

(event Custom|UpdateCurrentHeatValue
  (params (CurrentValue :type Float%20%28single-precision%29) (MaxValue :type Float%20%28single-precision%29) (State :type EFortWeaponOverheatState%20Enum) (bIsInCooldown :type Boolean))
  ;[1 1 1] Replicates the heat value and calls FX scripting on rep notify.
  (Variables|Default|SetCurrentHeatValueRep CurrentValue))

(event Custom|SetAudioPerspective
  ;[1 1 1] Audio Perspective
  (Audio|Components|Audio|SetSound (Variables|Default|GetOverheatLoopSound) (select (HelperFunctions|IsLocalPawn (Game|GetInstigator)) (Variables|Default|GetOverheatLoopSound_1P) (Variables|Default|GetOverheatLoopSound_3P))))

(event EventOnAmmoCountChanged
  (|CallParentFunction)
  (Variables|Default|SetCurrentAmmo (FortWeapon|GetLocalMagazineAmmoCount)))

(event Overheating|EventOnStartOverheated
  ;[1 1 1] Add overheat FX triggers here!
  (|CallParentFunction)
  (Development|PrintString "OVERHEATING")
  (CallFunction|SetScalarValueonWeaponMaterials "PowerUp_Cooldown" 1.0)
  (CallFunction|SetScalarValueonWeaponMaterials "PowerUp_CooldownTime" (Utilities|Time|GetGameTimeinSeconds))
  (Variables|Default|SetbIsInCooldown true)
  (Components|Activation|Activate (Variables|Default|GetNS_Powerup_MiniGun_Overheat) false)
  (Audio|Components|Audio|Play (Variables|Default|GetOnOverheatSound) 0.0)
  (CallFunction|OnEndHeatEvent))

(event Custom|Start%20Spin%20Up%20Audio
  ;[1 1 1] Handle volume scaling on the spin up and spin down audio cues
  (Audio|Components|Audio|FadeOut (Variables|Default|GetSpinDownSound) 0.25 0.0 "Linear" :!class /Script/Engine.AudioComponent)
  (bind _output_get (Variables|Default|SetCurrentSpinAudioComponent (Variables|Default|GetSpinUpSound)))
  (if (not (Variables|Default|GetbPlayedSpinUpAudio))
    (Audio|Components|Audio|Play _output_get 0.0)
    (Audio|Components|Audio|SetVolumeMultiplier _output_get 0.05 :!class /Script/Engine.AudioComponent)
    (Variables|Default|SetbPlayedSpinUpAudio true))
  (bind (_direction _volume) (Timeline|SpinAudioVolumeTimeline Play)
    (:Update
      (Variables|Default|SetSpinVolumeMultiplier _volume)
      (Audio|Components|Audio|SetVolumeMultiplier (Variables|Default|GetCurrentSpinAudioComponent) (Variables|Default|GetSpinVolumeMultiplier) :!class /Script/Engine.AudioComponent))))

(event EventOnChargeUp
  ;[1 1 1] OnStartCharge - Sound
  (|CallParentFunction)
  (CallFunction|StartSpinUpAudio)
  (Variables|Default|SetbPlayedSpinDownAudio false))

(event Custom|Start%20Spin%20Down%20Audio
  ;[1 1 1] Handle volume scaling on the spin up and spin down audio cues
  (Audio|Components|Audio|FadeOut (Variables|Default|GetSpinUpSound) 0.25 0.0 "Linear" :!class /Script/Engine.AudioComponent)
  (bind _output_get (Variables|Default|SetCurrentSpinAudioComponent (Variables|Default|GetSpinDownSound)))
  (if (not (Variables|Default|GetbPlayedSpinDownAudio))
    (Audio|Components|Audio|Play _output_get 0.0)
    (Audio|Components|Audio|SetVolumeMultiplier _output_get 0.05 :!class /Script/Engine.AudioComponent)
    (Variables|Default|SetbPlayedSpinDownAudio true))
  (Timeline|SpinAudioVolumeTimeline Reverse))

(event EventOnChargeDown
  ;[1 1 1] OnEndCharge - Sound
  (|CallParentFunction)
  (CallFunction|StartSpinDownAudio)
  (Variables|Default|SetbPlayedSpinUpAudio false))

(event EventOnPlayWeaponFireFX
  (params (bPersistentFire :type Boolean) (bSecondaryFire :type Boolean))
  ;[1 1 1] OnPlayWeaponFireFX / OnStopWeaponFireFX
  (|CallParentFunction bPersistentFire false)
  (CallFunction|SetScalarValueonWeaponMaterials "PowerUp_FireTime" (Utilities|Time|GetGameTimeinSeconds))
  (Audio|Components|Audio|Play (Variables|Default|GetOverheatLoopSound) 0.0)
  (Utilities|IsValid (Variables|Default|GetNS_MiniGun_MF)
    (:Is%20Valid
      (Components|Activation|Activate (Variables|Default|GetNS_MiniGun_MF) false))
    (:Is%20Not%20Valid
      (bind _returnvalue (Effects|SpawnFXSystemAttached :FXSystemAsset "/CorruptionItems/Gameplay/Items/PowerUp/Effects/NS_Powerup_MiniGun_MF.NS_Powerup_MiniGun_MF" :AttachToComponent (Variables|Visual|GetWeaponMesh) :AttachPointName "Muzzle" :Location "-40.000000, 0, 0" :Scale "1.000000,1.000000,1.000000" :LocationType "SnapToTarget" :bAutoDestroy true :PoolingMethod "None" :bAutoActivate true))
      (Variables|Default|SetNS_MiniGun_MF _returnvalue))))

(event EventOnStopWeaponFireFX
  ;[1 1 1] OnPlayWeaponFireFX / OnStopWeaponFireFX
  (|CallParentFunction)
  (Components|Activation|Deactivate (Variables|Default|GetNS_MiniGun_MF)))

(event Custom|BarrelOverheatFXandSound
  (bind _returnvalue (CallFunction|GetCurrentHeatPercent))
  (bind _returnvalue_2 (Math|Float|Lerp 0.0 1.0 _returnvalue))
  (bind _overheatsmoothed (Variables|Default|GetOverheatSmoothed))
  (bind _returnvalue_3 (Math|Float|MapRangeClamped _overheatsmoothed 0.08 1.0 0.0 1.0))
  ;[1 1 1] BarrelGlow - Handles ramping overheat FX and sound.  Timer shuts off when there is no heat in the minigun
  (Networking|CanExecuteCosmeticEvents
    (:True
      (CallFunction|SetScalarValueonWeaponMaterials "PowerUp_HeatAmount" _returnvalue_2)
      (Variables|Default|SetOverheatSmoothed (Math|Interpolation|FInterpTo (Variables|Default|GetOverheatSmoothed) _returnvalue_2 (Utilities|Time|GetWorldDeltaSeconds) 1e+01))
      (Audio|Parameter|SetFloatParameter (Variables|Default|GetOverheatLoopSound) "Overheat" _returnvalue_3 :!class /Script/AudioExtensions.AudioParameterControllerInterface)
      (Effects|Components|ParticleSystem|SetFloatParameter (Variables|Default|GetNS_Powerup_MiniGun_Idle) "User.Overheat" _returnvalue_3))))

(event FortWeapon|EventOnEquippedAndAttached
  ;[1 1 1] OnEquip - Reset MuzzleGlowMID Opacity
  (|CallParentFunction)
  (Networking|CanExecuteCosmeticEvents
    (:True
      (Audio|Parameter|SetFloatParameter (Variables|Default|GetOverheatLoopSound) "Overheat" 0.0 :!class /Script/AudioExtensions.AudioParameterControllerInterface)
      (Effects|Components|ParticleSystem|SetFloatParameter (Variables|Default|GetNS_Powerup_MiniGun_Idle) "User.Overheat" 0.0))))

(event FortWeapon|EventOnLocalUnEquip
  ;[1 1 1] Clear Timer Reset MuzzleGlowMID Opacity/Sound.  Keeps a phantom sound from playing.
  (|CallParentFunction)
  (Networking|CanExecuteCosmeticEvents
    (:True
      (Audio|Parameter|SetFloatParameter (Variables|Default|GetOverheatLoopSound) "Overheat" 0.0 :!class /Script/AudioExtensions.AudioParameterControllerInterface)
      (Effects|Components|ParticleSystem|SetFloatParameter (Variables|Default|GetNS_Powerup_MiniGun_Idle) "User.Overheat" 0.0))))

(event EventOnInitCosmeticAlterations
  (params (CosmeticMod :type Fort%20Cosmetic%20Modification%20Structure)))

(event Custom|OnStartHeatEvent
  ;[1 1 1] Heat Event logic
  ;;[1 1 1] Server Only. Handles heat event and reps bool to clients
  (if (not (Variables|Default|GetbIsInHeatEvent))
    (Variables|Default|SetbIsInHeatEvent true)
    (Networking|ForceNetUpdate)))

(event Custom|OnStartHeatEvent_Rep
  ;[1 1 1] Heat Event logic
  ;;[1 1 1] On Start Heat Event
  (Networking|CanExecuteCosmeticEvents
    (:True
      (FortWeapon|SetOverrideFireModeData (Variables|Default|GetHeatEventFireModeData))
      (CallFunction|SetScalarValueonWeaponMaterials "PowerUp_Heat" 1.0)
      (CallFunction|SetScalarValueonWeaponMaterials "PowerUp_HeatTime" (Utilities|Time|GetGameTimeinSeconds)))))

(event Custom|OnEndHeatEvent
  ;[1 1 1] Heat Event logic
  ;;[1 1 1] Server Only. Handles heat event and reps bool to clients
  (if (Variables|Default|GetbIsInHeatEvent)
    (Variables|Default|SetbIsInHeatEvent false)
    (Networking|ForceNetUpdate)))

(event Custom|OnEndHeatEvent_Rep
  ;[1 1 1] Heat Event logic
  ;;[1 1 1] OnEndHeatEvent
  (Networking|CanExecuteCosmeticEvents
    (:True
      (FortWeapon|ClearOverrideFireModeData)
      (if (not (Overheat|IsOverheated))
        ;;;[1 1 1] On End Heat Event. Naturally Cooling Down.
        (CallFunction|SetScalarValueonWeaponMaterials "PowerUp_Heat")
        (CallFunction|SetScalarValueonWeaponMaterials "PowerUp_HeatTime" (Utilities|Time|GetGameTimeinSeconds))))))

(event Overheating|EventOnFinishedCooling
  ;[1 1 1] On Completely Cooled off. No Heat.
  (|CallParentFunction)
  (Networking|CanExecuteCosmeticEvents
    (:True
      (Utilities|FlowControl|Sequence
        (:then_0
          (if (Variables|Default|GetbIsInCooldown)
            (CallFunction|SetScalarValueonWeaponMaterials "PowerUp_Cooldown")
            (CallFunction|SetScalarValueonWeaponMaterials "PowerUp_CooldownTime" (Utilities|Time|GetGameTimeinSeconds))
            (CallFunction|SetScalarValueonWeaponMaterials "PowerUp_Heat")
            (CallFunction|SetScalarValueonWeaponMaterials "PowerUp_HeatTime" (Utilities|Time|GetGameTimeinSeconds))
            (Variables|Default|SetbIsInCooldown false)))
        (:then_1
          (if (Audio|Components|Audio|IsPlaying (Variables|Default|GetOverheatLoopSound))
            (Audio|Components|Audio|FadeOut (Variables|Default|GetOverheatLoopSound) 1.0 0.0 "Linear" :!class /Script/Engine.AudioComponent)))))))

(fn GetWeaponMaterials
  (bind _weaponmesh (Variables|Visual|GetWeaponMesh))
  (for (_array_element _array_index) (Rendering|Material|GetMaterials _weaponmesh)
    (bind _returnvalue (Rendering|Material|CreateDynamicMaterialInstance _weaponmesh _array_index _array_element "None" :!class /Script/Engine.PrimitiveComponent))
    (Utilities|Array|Add (Variables|Default|GetWeaponMaterials) _returnvalue)))

(fn GetCurrentHeatPercent
  (return (/ (Variables|Default|GetCurrentHeatValueRep) (Overheat|GetOverheatingMaxValue))))

(fn OnRep_bIsInHeatEvent
  (if (Variables|Default|GetbIsInHeatEvent)
    (CallFunction|OnStartHeatEventRep)
    (else
      (CallFunction|OnEndHeatEventRep))))

(fn OnRep_CurrentHeatValueRep
  (CallFunction|BarrelOverheatFXandSound))

(fn SetScalarValueOnWeaponMaterials
  (params (Parameter :type Name) (Value :type Float%20%28double-precision%29))
  (for _array_element (Variables|Default|GetWeaponMaterials)
    (Rendering|Material|SetScalarParameterValue _array_element Parameter Value :!class /Script/Engine.MaterialInstanceDynamic)))

(fn ConstructionScript
  (Networking|CanExecuteCosmeticEvents
    (:True
      (CallFunction|GetWeaponMaterials))
    (:False
      _))
  (|CallParentFunction))
