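# Surface material for the skeletal LMG power-up weapon, with emissive crystal and barrel glow that reacts to heat, firing and cooldown.
# Heat, fire and cooldown parameters pair start timestamps with lengths, so the shader computes its own fades from Time.
# Named reroutes combine these fades with a sine rest pulse into intensities for the emissive mask's R and G channels, tinted by a heat-indexed curve atlas.
# One mask channel (G) drives metallic, specular and roughness.
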Material /CorruptionItems/Gameplay/Items/PowerUp/LMG/Materials/M_PowerUp_LMG
graph {
 properties {
  bUseMaterialAttributes=true
  bUsedWithMeshDeformer=true
  bUsedWithSkeletalMesh=true
  pixelDepthOffsetMode=PDOM_Legacy
  usedShadingModels=Default Lit
 }
 nodes [
  n1 = TextureSampleParameter2D() {parameterName=SpecularMasks, samplerType=SAMPLERTYPE_Masks, texture=/CorruptionItems/Gameplay/Items/PowerUp/LMG/Textures/T_PowerUp_LMG_1_S}
  n2 = TextureSampleParameter2D() {parameterName=Normal, samplerType=SAMPLERTYPE_Normal, texture=/CorruptionItems/Gameplay/Items/PowerUp/LMG/Textures/T_PowerUp_LMG_1_N}
  MakeMaterialAttributes!(
   BaseColor=TextureSampleParameter2D() {parameterName=Diffuse, texture=/CorruptionItems/Gameplay/Items/PowerUp/LMG/Textures/T_PowerUp_LMG_1_D}.RGB
   Metallic=@n1.G
   Specular=@n1.G
   Roughness=@n1.G
   Anisotropy
   EmissiveColor=NamedRerouteUsage()
   Opacity
   OpacityMask
   Normal=@n2.RGB
   Tangent
   WorldPositionOffset
   SubsurfaceColor
   ClearCoat
   ClearCoatRoughness
   AmbientOcclusion
   Refraction
   CustomizedUVs_0
   CustomizedUVs_1
   CustomizedUVs_2
   CustomizedUVs_3
   CustomizedUVs_4
   CustomizedUVs_5
   CustomizedUVs_6
   CustomizedUVs_7
   PixelDepthOffset
   ShadingModel
   Displacement
  )
  NamedRerouteDeclaration(None=LinearInterpolate(
    A=ScalarParameter() {defaultValue=1, parameterName=PowerUp_RestFireStrength}
    B=ScalarParameter() {defaultValue=10, parameterName=PowerUp_HeatFireStrength}
    Alpha=NamedRerouteUsage()
   )) {name=PulseStrength, nodeColor=<rgba 1 0.5156 0 1>}
  NamedRerouteDeclaration(None=Multiply(A=NamedRerouteUsage(), B=NamedRerouteUsage())) {name=BarrelEmissiveIntensity, nodeColor=<rgba 1 0 0.4219 1>}
  n3 = Time()
  n4 = ScalarParameter() {parameterName=PowerUp_HeatTime}
  n5 = Saturate(None=Divide(A=Subtract(A=@n3, B=@n4), B=ScalarParameter() {defaultValue=1, parameterName=PowerUp_HeatTimeLength}))
  NamedRerouteDeclaration(None=LinearInterpolate(A=OneMinus(None=@n5), B=@n5, Alpha=ScalarParameter() {parameterName=PowerUp_Heat})) {name=HeatValue, nodeColor=<rgba 0.4062 1 0 1>}
  n6 = Subtract(A=Time(), B=ScalarParameter() {group=PowerUp, parameterName=PowerUp_CooldownTime})
  n7 = Saturate(None=Divide(A=@n6, B=ScalarParameter() {defaultValue=2, group=PowerUp, parameterName=PowerUp_CooldownLength}))
  n8 = ScalarParameter() {group=PowerUp, parameterName=PowerUp_Cooldown}
  NamedRerouteDeclaration(None=LinearInterpolate(A=@n7, B=OneMinus(None=@n7), Alpha=@n8)) {name=CooldownValue, nodeColor=<rgba 0.125 0 1 1>}
  n9 = NamedRerouteUsage()
  n10 = Divide(
   A=Subtract(A=Time(), B=ScalarParameter() {group=PowerUp, parameterName=PowerUp_FireTime})
   B=ScalarParameter() {defaultValue=1, group=PowerUp, parameterName=PowerUp_FireTimeLength}
  )
  n11 = Multiply(A=OneMinus(None=Saturate(None=@n10)), B=NamedRerouteUsage())
  NamedRerouteDeclaration(None=Multiply(A=Add(A=@n9, B=@n11), B=NamedRerouteUsage())) {name=EmissiveIntensity, nodeColor=<rgba 0 0.7891 1 1>}
  n12 = TextureSampleParameter2D() {parameterName=Emissive, samplerType=SAMPLERTYPE_LinearColor, texture=/CorruptionItems/Gameplay/Items/PowerUp/LMG/Textures/T_PowerUp_LMG_1_E}
  n13 = Multiply(A=NamedRerouteUsage(), B=@n12.R)
  n14 = CurveAtlasRowParameter(CurveTime=ScalarParameter() {parameterName=PowerUp_HeatAmount}) {
   atlas=/CorruptionItems/Gameplay/Items/PowerUp/LMG/Materials/Curves/CA_PowerUpCurveAtlas
   curve=/CorruptionItems/Gameplay/Items/PowerUp/LMG/Materials/Curves/C_PowerUpGlowCurve
   parameterName=GlowCurve
  }
  n15 = Multiply(A=@n12.G, B=NamedRerouteUsage())
  NamedRerouteDeclaration(None=Add(A=Multiply(A=@n13, B=@n14), B=Multiply(A=@n15, B=@n14))) {name=Emissive, nodeColor=<rgba 0.6406 1 0 1>}
  n16 = Multiply(A=Time(), B=ScalarParameter() {defaultValue=0.5, parameterName=PowerUp_RestPulseSpeed})
  n17 = Multiply(A=ConstantBiasScale(None=Sine(None=@n16)), B=ScalarParameter() {defaultValue=1, parameterName=PowerUp_RestPulseIntensity})
  n18 = ScalarParameter() {defaultValue=2, parameterName=PowerUp_CrystalGlowOverheatBias}
  n19 = LinearInterpolate(
   A=ScalarParameter() {defaultValue=1, parameterName=PowerUp_CrystalGlowMin}
   B=Subtract(A=ScalarParameter() {defaultValue=10, parameterName=PowerUp_CrystalGlowMax}, B=@n18)
   Alpha=ScalarParameter() {parameterName=PowerUp_HeatAmount}
  )
  n20 = NamedRerouteUsage()
  NamedRerouteDeclaration(None=Add(A=Add(A=@n17, B=@n19), B=Multiply(A=@n18, B=@n20))) {name=GlowAmount, nodeColor=<rgba 0.6127 0 1 1>}
  n21 = ScalarParameter() {parameterName=PowerUp_BarrelGlowMin}
  n22 = ScalarParameter() {defaultValue=1, parameterName=PowerUp_BarrelGlowOverheatBias}
  n23 = Subtract(A=ScalarParameter() {defaultValue=2, parameterName=PowerUp_BarrelGlowMax}, B=@n22)
  n24 = ScalarParameter() {parameterName=PowerUp_HeatAmount}
  n25 = NamedRerouteUsage()
  NamedRerouteDeclaration(None=Add(A=LinearInterpolate(A=@n21, B=@n23, Alpha=@n24), B=Multiply(A=@n22, B=@n25))) {name=BarrelGlowAmount, nodeColor=<rgba 0 0.1562 1 1>}
  n26 = Saturate(None=Divide(A=@n6, B=ScalarParameter() {defaultValue=2, group=PowerUp, parameterName=PowerUp_BarrelCooldownLength}))
  NamedRerouteDeclaration(None=LinearInterpolate(A=@n26, B=OneMinus(None=@n26), Alpha=@n8)) {name=BarrelCooldown, nodeColor=<rgba 0 0.5312 1 1>}
 ]
}
